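# Niagara idle effect for the UniqueBeetle bass cosmetic in Sparks, adding persistent pink glows to the instrument.
# CapsuleGlow bursts five particles, using the partition index to set each one's fixed position, size and alpha: one low, two mirrored pairs higher.
# Sprite size follows owner scale, while color is multiplied by a system time-of-day color and alpha is dimmed in the front end.
# GlowMesh and Light emitters list no renderers.
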NiagaraSystem /SparksCosmetics/Instrument/Bass/UniqueBeetle/Effects/NS_Bass_UniqueBeetle_Idle_FX
dependencies {
 usedDataInterfaces [&NiagaraDataInterfaceCurve, &NiagaraDataInterfaceSocketReader]
 usedDynamicInputs [
  &MergeHelper_Vector2Switch
  &MakeCustomLinearColorFromBool
  &FNTimeOfDayColor
  &MakeFloatFromVector
  &FloatFromCurve
 ]
 usedModules [
  &self:SystemSpawnScript.SystemScriptSource.SystemScriptGraph.NiagaraNodeAssignment_11.SetVariables_#1
  &self:SystemSpawnScript.SystemScriptSource.SystemScriptGraph.NiagaraNodeAssignment_0.SetVariables_#2
  &SystemState
  &self:SystemSpawnScript.SystemScriptSource.SystemScriptGraph.NiagaraNodeAssignment_10.SetVariables_#3
  &self:SystemSpawnScript.SystemScriptSource.SystemScriptGraph.NiagaraNodeAssignment_2.SetVariables_#4
  &self:SystemSpawnScript.SystemScriptSource.SystemScriptGraph.NiagaraNodeAssignment_9.SetVariables_#5
 ]
}
emitters {
 CapsuleGlow {
  emitterUpdateScript [
   &EmitterState(Life Cycle Mode=::NewEnumerator1, Loop Behavior=::NewEnumerator1, Loop Duration Mode=::NewEnumerator1, UseLoopDelay=false)
   &SpawnBurst_Instantaneous(Spawn Count=5, Use Spawn Probability=false, UseLoopCountLimit=true)
  ]
  particleSpawnScript [
   &PartitionParticles(ID to Partition:NiagaraInt32=&ReturnExecIndex(), Number Of Partitions=5)
   &InitializeParticle(
    Color=&MergeHelper_ColorSwitch(
     0=<rgba 1 0.0975 0.5413 0.3>
     1=<rgba 1 0.0975 0.5413 0.15>
     2=<rgba 1 0.0975 0.5413 0.15>
     3=<rgba 1 0.0975 0.5413 0.15>
     4=<rgba 1 0.0975 0.5413 0.15>
     Index=$Particles.PartitionParticles.Partition
     Num=::NewEnumerator3
    )
    Position=&MergeHelper_VectorSwitch(
     0=<0 0 1.2>
     1=<6 0.25 88.65>
     2=<-6 0.25 88.65>
     3=<7 0.25 84>
     4=<-7 0.25 84>
     Index=$Particles.PartitionParticles.Partition
     Num=::NewEnumerator3
    )
    Position Mode=::NewEnumerator1
    Sprite Size Mode=::NewEnumerator3
    Uniform Sprite Size=&MergeHelper_FloatSwitch(0=10, 1=7.5, 2=7.5, 3=7.5, 4=7.5, Index=$Particles.PartitionParticles.Partition, Num=::NewEnumerator3)
    UsePositionOffset=false
    Write Material Random=true
   )
   &CameraOffset(Camera Offset Amount=2)
  ]
  particleUpdateScript [
   &ParticleState(Kill Particles When Lifetime Has Elapsed=false, Let Infinitely Lived Particles Die When Emitter Deactivates=true)
   &ScaleSpriteSize(Scale Sprite Size Mode=::NewEnumerator0, Uniform Scale Factor=$System.EngineOwnerScale)
   &DynamicMaterialParameters(
    Index 0 Param 1=1
    Index 0 Param 2=10
    Module.Index0Param0Write=true
    Module.Index0Param1Write=false
    Module.Index0Param2Write=false
    Module.Index0Param3Write=false
   ) {module=DynamicMaterialParameters001}
   &ScaleColor(
    Color Scale:LinearColor=&MultiplyLinearColor(
     Linear Color=$System.TODColor
     Scale Factor=&MakeVector4FromVectorAndFloat("Vector (RGB)"=<1 1 1>, "Float (Alpha)"=&FN_FrontEndFloatModifier(Gameplay=1, Front End=0.7))
    )
    Scale Mode=3
   ) {module=ScaleColor001}
  ]
  rendererClasses [&NiagaraSpriteRendererProperties, &NiagaraMeshRendererProperties]
  renderers [
   {rendererClass=&NiagaraSpriteRendererProperties}
   {rendererClass=&NiagaraMeshRendererProperties, rendererIndex=1}
  ]
 }
 GlowMesh {rendererClasses [], renderers []}
 Light {rendererClasses [], renderers []}
}
summary {
 systemName=NS_Bass_UniqueBeetle_Idle_FX
 emitters [
  {emitterName=CapsuleGlow, rendererClasses [&NiagaraSpriteRendererProperties, &NiagaraMeshRendererProperties]}
  {emitterName=GlowMesh, rendererClasses []}
  {emitterName=Light, rendererClasses []}
 ]
 userVariables [
  User.bIsFrontEnd {type=NiagaraBool}
  User.bIsFrontEndPreview {type=NiagaraBool}
 ]
}
