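# Niagara ribbon swing trail for the UniqueBeetle bass pickaxe in the Sparks instrument cosmetics.
# One emitter, TrailBladeMain, uses the FN anim-trail modules to spawn, initialize and update short-lived ribbon particles between two bone or socket indices.
# Spawn alpha follows System.FadeSpawn, then ScaleColor applies System.TODColor and System.FadeSystem, with time-of-day color among the system's inputs.
# AddVelocity and the force solver are disabled.
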NiagaraSystem /SparksCosmetics/Instrument/Bass/UniqueBeetle/Effects/NS_Pickaxe_Bass_UniqueBeetle_SwingTrail
dependencies {
 usedDataInterfaces [&NiagaraDataInterfaceSocketReader, &NiagaraDataInterfaceColorCurve]
 usedDynamicInputs [
  &FNTimeOfDayColor
  &ClampFloat
  &RemapRange
  &Subtract_Float
  &Add_Float
 ]
 usedModules [
  &self:SystemSpawnScript.SystemScriptSource.SystemScriptGraph.NiagaraNodeAssignment_3.SetVariables_#1
  &self:SystemSpawnScript.SystemScriptSource.SystemScriptGraph.NiagaraNodeAssignment_4.SetVariables_#2
  &SystemState
  &self:SystemSpawnScript.SystemScriptSource.SystemScriptGraph.NiagaraNodeAssignment_2.SetVariables_#3
  &self:SystemSpawnScript.SystemScriptSource.SystemScriptGraph.NiagaraNodeAssignment_0.SetVariables_#4
  &self:SystemSpawnScript.SystemScriptSource.SystemScriptGraph.NiagaraNodeAssignment_1.SetVariables_#5
  &self:SystemSpawnScript.SystemScriptSource.SystemScriptGraph.NiagaraNodeAssignment_5.SetVariables_#6
 ]
}
emitters {TrailBladeMain {
  emitterUpdateScript [
   &EmitterState()
   &FN_SpawnAnimTrail()
  ]
  particleSpawnScript [
   &FN_InitializeAnimTrailBoneIndecies(First Bone or Socket Index=1, Second Bone or Socket Index=0)
   &FN_InitializeAnimTrail(
    Color=&MakeLinearColorFromVectorAndFloat(
     "Vector (RGB)"=&MakeVectorFromLinearColorRGB(Color=<rgba 1 1 1 1>)
     "Float (Alpha)"=&Multiply_Float(A=1, B=$System.FadeSpawn)
    )
    Lifetime=0.15
   )
   &AddVelocity(Velocity=<0 900 0>) {enabled=false}
  ]
  particleUpdateScript [
   &FN_UpdateAnimTrail() {module=FN_UpdateAnimTrail001}
   &ScaleColor(
    Scale Mode=1
    Scale RGBA=&Multiply_Vector4(
     A=&MakeVector4_FromLinearColor(LinearColor=$System.TODColor)
     B=&MakeVector4FromVectorAndFloat("Vector (RGB)"=<1 1 1>, "Float (Alpha)"=$System.FadeSystem)
    )
    ScaleRGBA=true
   ) {module=ScaleColor002}
   &SolveForcesAndVelocity(Clamp Velocity=false, Limit Acceleration=false) {enabled=false}
  ]
  rendererClasses [&NiagaraRibbonRendererProperties]
  renderers [{rendererClass=&NiagaraRibbonRendererProperties}]
 }}
summary {
 systemName=NS_Pickaxe_Bass_UniqueBeetle_SwingTrail
 emitters [{emitterName=TrailBladeMain, rendererClasses [&NiagaraRibbonRendererProperties]}]
 userVariables [
  User.RibbonIDOffset {type=NiagaraInt32}
  User.SkeletalMesh {
   type=NiagaraDataInterfaceSkeletalMesh
   defaultValue {
    struct=NiagaraExt_VariableValue_DataInterface
    value {dataInterface=&self:NiagaraDataInterfaceSkeletalMesh_0}
   }
  }
  User.SwingSectionIndex {type=NiagaraInt32}
 ]
}
